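# Niagara effect for the Chicken Warrior backpack that drips slurp sprites from its skeletal mesh, both idle and on landing.
# Idle_Drips spawns sparse random drops that fall under ramping gravity, stretch and fade out.
# Jump_Drips scales its spawn rate by User.Land, offsets drops randomly around the mesh and inherits amplified mesh velocity.
# Both animate dynamic material parameters over particle life and use camera offsets.
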NiagaraSystem /BRCosmetics/Accessories/FORT_Backpacks/Backpack_M_MED_Chicken_Warrior/FX/NS_Backpack_ChickenWarrior_Slurp
dependencies {
 usedDataInterfaces [&NiagaraDataInterfaceCurve, &NiagaraDataInterfaceVector2DCurve, &NiagaraDataInterfaceArrayFloat]
 usedDynamicInputs []
 usedModules [&SystemState]
}
emitters {
 Idle_Drips {
  emitterUpdateScript [
   &EmitterState(Life Cycle Mode=::NewEnumerator1, Recalculate Duration Each Loop=true, UseLoopDelay=false)
   &SpawnRate(
    Spawn Probability=0.2
    SpawnRate=&RandomRangeFloat(Minimum=3, Maximum=6, Evaluation Type=::NewEnumerator1)
    Use Spawn Probability=true
   )
   &SpawnBurst_Instantaneous(
    Spawn Count=1
    Spawn Probability=&FN_FrontEndFloatModifier(Gameplay=&MakeCustomFloatFromBool(Bool=$User.IsOnGround, True Float=1, False Float=0), Front End=1)
    Use Spawn Probability=true
   ) {enabled=false}
  ]
  particleSpawnScript [
   /Niagara/Modules/Spawn/Initialization/InitializeParticle(
    Color=&MakeLinearColorFromVectorAndFloat(
     "Vector (RGB)"=&MakeVectorFromLinearColorRGB(Color=<rgba 1 1 1 1>)
     "Float (Alpha)"=&RandomRangeFloat(Minimum=0.5, Maximum=0.8)
    )
    Module.Write Color=true
    Module.Write Lifetime=true
    Module.Write Mass=true
    Module.Write Material Random=true
    Module.Write Position=false
    Module.Write Scale=true
    Module.Write SpriteRotation=false
    Module.Write SpriteSize=true
    Position=&SimulationPosition()
    Sprite Size=&FN_FrontEndVector2Modifier(Front End:Vector2f=<3 5>, Gameplay:Vector2f=<4 7>)
   )
   &SkeletalMeshLocation(
    Inherit Velocity=false
    MaskSpawnGroup=false
    Mesh Sampling Type=::NewEnumerator5
    Sampled Position Offset Coordinate Space:ENiagaraMeshTransforms=::NewEnumerator3
    Skeletal Mesh=$User.SkeletalMesh
    Socket Sampling Mode=::NewEnumerator5
   )
   &SphereLocation(
    Hemisphere Z=true
    MaskSpawnGroup=false
    Non Uniform Scale=<1 0.8 -1>
    Override Seed=false
    Sphere Radius=9
    Surface Only=true
    Surface Only Band Thickness=1
    UseNonUniformScale=true
    UseOffset=false
    UseSphereOrientation=false
   ) {enabled=false}
   &InheritVelocity(Module.Limit Speed=true, Velocity Scale=&Multiply_VectorByFloat(Vector=<1 1 1>, Float=0.2)) {enabled=false, module=InheritVelocity001}
   &AddVelocity(Velocity=<0 0 -20>) {enabled=false}
  ]
  particleUpdateScript [
   &ParticleState()
   &GravityForce(Gravity=&Multiply_VectorByFloat(
     Vector=<1 1 -600>
     Float=&FloatFromCurve(FloatCurve=DataInterface() {propertyValues {Curve {keys [{arriveTangent=-0.0925, interpMode=RCIM_Cubic, leaveTangent=3.9937, tangentMode=RCTM_Break, time=0.2, value=0}, <at 1 1>]}}})
    ))
   &ScaleSpriteSize(Scale Factor=&Vector2DFromCurve(
     Vector2Curve=DataInterface() {propertyValues {
       XCurve {keys [<at 0.25 1 RCIM_Cubic>, {interpMode=RCIM_Cubic, leaveTangent=-2.9373, time=1, value=0.3}]}
       YCurve {keys [
         {arriveTangent=-1.4216, interpMode=RCIM_Cubic, time=0.0067, value=0.3563}
         {arriveTangent=2.9367, interpMode=RCIM_Cubic, leaveTangent=2.9367, tangentMode=RCTM_User, time=0.2953, value=1.3705}
         {interpMode=RCIM_Cubic, leaveTangent=7.4188, time=1, value=2}
        ]}
      }}
     ScaleCurve=<1 1>
    ))
   &SolveForcesAndVelocity(Clamp Velocity=false, Limit Acceleration=false)
   &DynamicMaterialParameters(
    Index 0 Param 1=0.8
    Index 0 Param 2=&FloatFromCurve(FloatCurve=DataInterface() {propertyValues {Curve {keys [{arriveTangent=-1.8884, interpMode=RCIM_Cubic, leaveTangent=-1.8884, tangentMode=RCTM_User, time=0, value=1}, <at 1 0.26>]}}})
    Index 0 Param 3=1
    Index 0 Param 4=1
    Module.Param0WriteEnabled=true
    Module.Param1WriteEnabled=true
    Module.Param2WriteEnabled=false
    Module.Param3WriteEnabled=false
   )
   &CameraOffset(Camera Offset Amount=&RandomRangeFloat(Minimum=0, Maximum=6))
   &ScaleColor(
    Scale Alpha=&FloatFromCurve(FloatCurve=DataInterface() {propertyValues {Curve {keys [<at 0 1>, {arriveTangent=-1, interpMode=RCIM_Cubic, leaveTangent=-1, time=0.8, value=1}, <at 1 0>]}}})
    ScaleA=true
    ScaleRGB=true
   )
  ]
  rendererClasses [&NiagaraSpriteRendererProperties]
  renderers [{rendererClass=&NiagaraSpriteRendererProperties}]
 }
 Jump_Drips {
  emitterUpdateScript [
   &EmitterState(Life Cycle Mode=::NewEnumerator1, Loop Duration=0.1, Recalculate Duration Each Loop=true, UseLoopDelay=true)
   &SpawnRate(SpawnRate=&Multiply_Float(A=10, B=$User.Land), Use Spawn Probability=true)
   /SetVariables(Emitter.PlayerVelocityVector:Vector3f=$Engine.Owner.Velocity) {isSetParameters=true}
   /SetVariables(Emitter.PlayerSpeed:NiagaraFloat=&VectorLength(Vector=$Emitter.PlayerVelocityVector)) {isSetParameters=true}
  ]
  particleSpawnScript [
   &InitializeParticle(
    Mesh Scale=&Multiply_VectorByFloat(
     Vector=&Multiply_VectorByFloat(
      Vector=<1 1 1>
      Float=&Multiply_Float(
       A=0.1
       B=&SelectFloatFromArray(Direct Array Index=0, Float Selection Array=DataInterface() {propertyValues {FloatData [0,1]}})
      )
     )
     Float=1
    )
    Module.Write Color=true
    Module.Write Lifetime=true
    Module.Write Mass=true
    Module.Write Material Random=true
    Module.Write Position=false
    Module.Write Scale=true
    Module.Write SpriteRotation=false
    Module.Write SpriteSize=true
    Position=&SimulationPosition()
    Sprite Size=<4 10>
   )
   &SkeletalMeshLocation(
    Bone Sampling Mode=::NewEnumerator4
    Inherit Velocity=true
    "Inherit Velocity (Scale)"=10
    MaskSpawnGroup=false
    Sampled Position Offset=&Add_Vector(
     A=<-2.5 -11 0>
     B=&Multiply_Vector(A=&RandomVector(Vector Scale=11), B=<1 1 1.2>)
    )
    Sampled Position Offset Coordinate Space:ENiagaraMeshTransforms=::NewEnumerator3
    Skeletal Mesh=$User.SkeletalMesh
    VelocitySpeedLimit=false
   )
   /SetVariables(Particles.LocalPosition:Vector3f=/Niagara/DynamicInputs/Transforms/TransformPosition(Position=Linked() {linkedVariable {name=Particles.Position, type=NiagaraPosition}}, Source Space=::World, Destination Space=::Local)) {enabled=false, isSetParameters=true}
   &ApplyInitialForces()
   &InheritVelocity(Module.Limit Speed=true, Velocity Scale=&Multiply_VectorByFloat(Vector=<1 1 1>, Float=1)) {enabled=false, module=InheritVelocity001}
   /SetVariables() {isSetParameters=true}
  ]
  particleUpdateScript [
   &ParticleState()
   &GravityForce(Gravity=&Multiply_VectorByFloat(
     Vector=<1 1 -700>
     Float=&FloatFromCurve(FloatCurve=DataInterface() {propertyValues {Curve {keys [{arriveTangent=1.8039, interpMode=RCIM_Cubic, leaveTangent=1.8039, tangentMode=RCTM_User, time=0, value=0}, <at 1 1>]}}})
    ))
   &SolveForcesAndVelocity(Clamp Velocity=false, Limit Acceleration=false)
   &ScaleSpriteSize(Scale Factor=&Vector2DFromCurve(
     Vector2Curve=DataInterface() {propertyValues {
       XCurve {keys [
         <at 0 1.15 RCIM_Cubic>
         {arriveTangent=-0.8581, interpMode=RCIM_Cubic, leaveTangent=-0.8581, time=0.6037, value=1.14}
         {arriveTangent=-0.3533, interpMode=RCIM_Cubic, leaveTangent=-0.3533, time=0.8741, value=0.4}
         <at 1 1>
        ]}
       YCurve {keys [
         <at 0 0 RCIM_Cubic>
         {arriveTangent=2.8602, interpMode=RCIM_Cubic, leaveTangent=2.8602, time=0.6037, value=0.9885}
         {arriveTangent=2.0477, interpMode=RCIM_Cubic, leaveTangent=2.0477, time=0.8741, value=2.5}
         <at 1 1.8>
        ]}
      }}
     ScaleCurve=<1 1>
    ))
   &DynamicMaterialParameters(
    Index 0 Param 1=0.8
    Index 0 Param 2=&FloatFromCurve(FloatCurve=DataInterface() {propertyValues {Curve {keys [{arriveTangent=-1.8884, interpMode=RCIM_Cubic, leaveTangent=-1.8884, tangentMode=RCTM_User, time=0, value=1}, <at 1 0.26>]}}})
    Index 0 Param 3=1
    Index 0 Param 4=1
    Module.Param0WriteEnabled=true
    Module.Param1WriteEnabled=true
    Module.Param2WriteEnabled=true
    Module.Param3WriteEnabled=true
   )
   /SetVariables(Particles.Position=&Lerp_Vector(
     A=&TransformPosition(Position=$Particles.LocalPosition, Source Space=::Local, Destination Space=::Simulation)
     B=Linked() {linkedVariable {name=Particles.Position, type=NiagaraPosition}}
     Alpha=&FloatFromCurve(FloatCurve=DataInterface() {propertyValues {Curve {keys [<at 0.2 0 RCIM_Cubic>, {arriveTangent=7.4901, interpMode=RCIM_Cubic, tangentMode=RCTM_Break, time=0.6, value=1}]}}})
    )) {enabled=false, isSetParameters=true}
   &CameraOffset(Camera Offset Amount=5)
  ]
  rendererClasses [&NiagaraSpriteRendererProperties]
  renderers [
   {rendererClass=&NiagaraSpriteRendererProperties}
   {rendererClass=&NiagaraSpriteRendererProperties, rendererIndex=1}
  ]
 }
}
summary {
 systemName=NS_Backpack_ChickenWarrior_Slurp
 emitters [
  {emitterName=Idle_Drips, rendererClasses [&NiagaraSpriteRendererProperties]}
  {emitterName=Jump_Drips, rendererClasses [&NiagaraSpriteRendererProperties]}
 ]
 userVariables [
  User.bIsFrontEnd {type=NiagaraBool}
  User.bIsFrontEndPreview {type=NiagaraBool}
  User.Land {type=NiagaraFloat, defaultValue=0, description=""}
  User.SkeletalMesh {
   type=NiagaraDataInterfaceSkeletalMesh
   defaultValue {
    struct=NiagaraExt_VariableValue_DataInterface
    value {dataInterface=&self:NiagaraDataInterfaceSkeletalMesh_0}
   }
  }
 ]
}
