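# Physics asset for the dangling side pouches on the Chicken Warrior backpack cosmetic.
# Left and right pouch bones each get a tilted box body, while the pack bone holds a tiny capsule as the anchor.
# Each pouch is constrained to the pack with tight swing and twist limits, allowing only slight movement around it.
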PhysicsAsset /BRCosmetics/Accessories/FORT_Backpacks/Backpack_M_MED_Chicken_Warrior/Meshes/M_MED_Chicken_Warrior_Pack_Physics_Pouches
bodies {
 dyn_pouch_01_l [{shapeType=Box, extentX=3, extentY=8, extentZ=10, center=<1.5299 -0.0003 0>, rotation=<rot 14.9999 0 0>}]
 dyn_pouch_01_r [{shapeType=Box, extentX=3, extentY=8, extentZ=10, center=<1.6091 0.0001 -0.0241>, rotation=<rot 10.0003 0 0>}]
 pack [{shapeType=Capsule, length=1, radius=1}]
}
constraints [
 {bone1Name=dyn_pouch_01_l, bone2Name=pack, swing1LimitDegrees=5, swing2LimitDegrees=5, twistLimitDegrees=2}
 {bone1Name=dyn_pouch_01_r, bone2Name=pack, swing1LimitDegrees=5, swing2LimitDegrees=5, twistLimitDegrees=2}
]
