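# Physics asset for the side strap of the Chicken Warrior backpack cosmetic, defining the bodies and joints for its simulation.
# Five small box bodies form a chain from the capsule body on the strap master bone, each constrained to the previous with tight swing and twist limits.
# A larger box on the right pouch bone has no constraint, so it sits outside the chain.
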PhysicsAsset /BRCosmetics/Accessories/FORT_Backpacks/Backpack_M_MED_Chicken_Warrior/Meshes/M_MED_Chicken_Warrior_Pack_Physics_SideStrap
bodies {
 dyn_pouch_01_r [{shapeType=Box, extentX=3, extentY=8, extentZ=10, center=<1.45 0 -0.0545>, rotation=<rot 10.0003 0 0>}]
 dyn_sideStrap_01 [{shapeType=Box, extentX=3, extentY=1, extentZ=1, center=<1.5 0 0>}]
 dyn_sideStrap_02 [{shapeType=Box, extentX=3, extentY=1, extentZ=1, center=<1.5 0 0>}]
 dyn_sideStrap_03 [{shapeType=Box, extentX=3, extentY=1, extentZ=1, center=<1.5 0 0>}]
 dyn_sideStrap_04 [{shapeType=Box, extentX=3, extentY=1, extentZ=1, center=<1.5 0 0>}]
 dyn_sideStrap_05 [{shapeType=Box, extentX=2, extentY=1, extentZ=1, center=<1 0 0>}]
 dyn_sideStrap_mstr [{shapeType=Capsule, length=3, radius=9.25, center=<3.9556 -2.5618 -10.0579>, rotation=<rot -74.9902 179.9999 89.9951>}]
}
constraints [
 {bone1Name=dyn_sideStrap_01, bone2Name=dyn_sideStrap_mstr, swing1LimitDegrees=5, swing2LimitDegrees=5, twistLimitDegrees=2}
 {bone1Name=dyn_sideStrap_02, bone2Name=dyn_sideStrap_01, swing1LimitDegrees=5, swing2LimitDegrees=5, twistLimitDegrees=2}
 {bone1Name=dyn_sideStrap_03, bone2Name=dyn_sideStrap_02, swing1LimitDegrees=5, swing2LimitDegrees=5, twistLimitDegrees=2}
 {bone1Name=dyn_sideStrap_04, bone2Name=dyn_sideStrap_03, swing1LimitDegrees=5, swing2LimitDegrees=5, twistLimitDegrees=2}
 {bone1Name=dyn_sideStrap_05, bone2Name=dyn_sideStrap_04, swing1LimitDegrees=5, swing2LimitDegrees=5, twistLimitDegrees=2}
]
